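# Animation Blueprint for the Carbide Knight backpack, adjusting pack, scabbard and handle procedurally from the reference pose.
# dyn_handle shrinks away while the matching pickaxe is equipped, except during skydiving, which also offsets the pack.
# During standing melee, the scabbard rolls by the angle between the parent's pelvis and spine_05, smoothed on increase.
# Crouch offset and pelvis location still update, but their scabbard and look-at nodes are disconnected from the output.
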AnimBlueprint /BRCosmetics/Accessories/FORT_Backpacks/Backpack_M_MED_Carbide_Knight/Meshes/M_MED_Carbide_Knight_Pack_AnimBP
target_skeleton=&/M_MED_Carbide_Knight_Pack_Skeleton
graphs [
 AnimGraph {nodes [
   Local Space Ref Pose()
   n1 = Look At - Bone: pack!(ComponentPose, Alpha=1, LookAtLocation:Vector=Get PelvisLocation() -> (PelvisLocation:Vector)) -> (Pose) {
    bUseLookUpAxis=true
    alphaBoolBlend {blendOption=Linear}
    boneToModify {boneName=pack}
    lookAt_Axis {bInLocalSpace=false, axis {y=0, z=-1}}
    lookAtTarget=~
   }
   "Transform (Modify) Bone - Bone: scabbard"(ComponentPose=@n1, Alpha=1, Translation="0,0,0", Rotation=Get Crouching offset() -> (Crouching offset:Rotator), Scale="1,1,1") {
    rotationMode=BMM_Additive
    translationMode=BMM_Additive
    alphaBoolBlend {blendOption=Linear}
    boneToModify {boneName=scabbard}
    rotation {pitch=-50}
   }
   "Transform (Modify) Bone - Bone: pack"(Translation="0,-3,0", Rotation="0,0,20") {
    translationMode=BMM_Additive
    translationSpace=BCS_BoneSpace
    alphaBoolBlend {blendOption=Linear}
    boneToModify {boneName=pack}
    rotation {roll=20}
    translation {y=-3}
   }
   n2 = "Transform (Modify) Bone - Bone: dyn_handle"(
    ComponentPose=Local To Component(LocalPose=Local Space Ref Pose())
    Alpha=Get HideHandle() -> (HideHandle:double)
    Translation="-30,0,0"
    Rotation="0,0,0"
    Scale="0.01,0.01,0.01"
   ) {
    scaleMode=BMM_Replace
    scaleSpace=BCS_BoneSpace
    translationMode=BMM_Additive
    translationSpace=BCS_BoneSpace
    alphaBoolBlend {blendOption=Linear}
    boneToModify {boneName=dyn_handle}
    scale=<0.01 0.01 0.01>
    translation {x=-30}
   }
   n3 = "Transform (Modify) Bone - Bone: scabbard"(ComponentPose=@n2, Alpha=Get MeleeAlpha() -> (MeleeAlpha:double), Translation="0,0,0", Rotation="0,0,30", Scale="1,1,1") {
    rotationMode=BMM_Additive
    translationMode=BMM_Additive
    alphaBoolBlend {blendOption=Linear}
    alphaScaleBiasClamp {bInterpResult=true, interpSpeedIncreasing=5}
    boneToModify {boneName=scabbard}
    rotation {roll=30}
   }
   n4 = "Transform (Modify) Bone - Bone: pack"(ComponentPose=@n3, Alpha=Get SkyDivingCorrection() -> (SkyDivingCorrection:double), Translation="0,-2,0", Rotation="-15,0,0") {
    rotationMode=BMM_Additive
    translationMode=BMM_Additive
    alphaBoolBlend {blendOption=Linear}
    boneToModify {boneName=pack}
    rotation {pitch=-15}
    translation {y=-2}
   }
   Output Pose(Result=Component To Local(ComponentPose=@n4)) {layerGroup=None}
  ]}
 EventGraph {nodes [
   n1 = Event BlueprintUpdateAnimation()
   > n2 = Sequence!(execute) -> (then_0, then_1, then_2:Exec)
   > Set SkyDivingCorrection(SkyDivingCorrection:double=SelectFloat(A=1, B=0, bPickA=Get bIsSkyDiving())) -> (Output_Get:double)
   n3 = Get bIsCrouching()
   > n4 = Set Crouch_interpolate(Crouch_interpolate:double=FInterpTo(
     Current=Get Crouch_interpolate() -> (Crouch_interpolate:double)
     Target="To Float (Boolean)"(InBool=@n3)
     DeltaTime=@n1.DeltaTimeX
     InterpSpeed=10
    )) -> (Output_Get:double)
   > Set Crouching offset(Crouching offset:Rotator="Lerp (Rotator)"!(A:Rotator="10, 0,25", B:Rotator="5, 0,38", Alpha:float=@n4, bShortestPath:bool=false) -> (ReturnValue:Rotator)) -> (Output_Get:Rotator)
   n5 = GetPartSkeletalMeshComponent()
   Is Valid(exec=@n2.then_1, InputObject=@n5)
   n6 = GetAttachParent!(self:Scene Component Object Reference=@n5) -> (ReturnValue:Scene Component Object Reference)
   > Is Valid(InputObject=@n6)
   n7 = BreakTransform(InTransform=GetSocketTransform(self=@n6, InSocketName=pelvis, TransformSpace=RTS_World))
   > Set PelvisLocation(PelvisLocation:Vector=@n7) -> (Output_Get:Vector)
   n8 = Get bIsCrouching()
   n9 = Branch(execute=@n2.then_2, Condition=AND Boolean(A=Get bIsPlayingMeleeAnim(), B="Not Equal (Boolean)"(A=@n8, B=true) -> (ReturnValue=~)))
   n10 = Vector_Forward!() -> (ReturnValue:Vector)
   n11 = GetSocketTransform(self=@n6, InSocketName=spine_05, TransformSpace=RTS_World)
   n12 = GetAngleBetweenTwoVectors!(
    VectorA:Vector=Rotate Vector!(A:Vector=@n10, B:Rotator=@n7.Rotation) -> (ReturnValue:Vector)
    VectorB:Vector=Rotate Vector(A=@n10, B=BreakTransform(InTransform=@n11).Rotation)
   ) -> ("Angle (Degrees)":double)
   > Set MeleeAlpha(MeleeAlpha:double=NormalizeToRange!(Value:double=float - float(A=@n12, B=5), RangeMin:double=0, RangeMax:double=18) -> (ReturnValue:double)) -> (Output_Get:double)
   Set MeleeAlpha(execute=@n9.else, MeleeAlpha=0)
   n13 = Event OnWeaponChanged()
   n14 = Sequence!(execute=[@n13.then, @n21.then]) -> (then_0, then_1)
   n15 = PickaxeItemDefinitionHasAnyMetaTag(Tags=Make Array!("[0]"=(TagName="Cosmetics.MetaTags.CarbideKnightMale"), "[1]":Gameplay Tag Structure) -> (Array), bRequiresEquipped=true)
   > n16 = Branch(Condition=@n15)
   > n17 = Branch(Condition=Get bIsSkyDiving())
   Set HideHandle(execute=[@n19.then, @n17.then, @n18.else, @n16.else], HideHandle:double=0) -> (Output_Get:double)
   Set HideHandle(execute=[@n17.else, @n18.then], HideHandle=1)
   n18 = Branch(execute=[@n20.then, @n14.then_1], Condition=@n15)
   n19 = Event OnBeginSkydiving()
   n20 = Event OnEndSkydiving()
   n21 = Event OnCharacterCustomizationLoaded()
  ]}
]
variables [
 Crouch_interpolate {type=number}
 Crouching offset {type=Rotator}
 HideHandle {type=number, default=1}
 MeleeAlpha {type=number}
 PelvisLocation {type=Vector}
 SkyDivingCorrection {type=number}
]
